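; Blueprint AI spawner component for the FrenchYedoc NPC's affiliation, judging by its name, overriding the squad ID lookup.
; GetSquadID returns the value of GetCustomSquadID along with a true flag, so the custom squad ID is always reported as valid or used.
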(fn GetSquadID
  (return (Variables|Default|GetCustomSquadID) true))
